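# Niagara system for the Obsidian emote that plays a small debris explosion of dark rocks and dusty smoke.
# Rocks bursts mesh chunks outward along sphere normals, then gravity pulls them down as they spin and shrink away late in life.
# Wispy_Smoke bursts tan-tinted sprites the same way, with drag, gravity, growing size, fading alpha and SubUV flipbook animation.
# The two emitters run independently, and the system exposes no user parameters.
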NiagaraSystem /BRCosmetics/Effects/Fort_Effects/Emotes/Obsidian/NS_Obsidian_Emote_DebrisExplosion
dependencies {usedDataInterfaces [&NiagaraDataInterfaceCurve], usedDynamicInputs [], usedModules [&SystemState]}
emitters {
 Rocks {
  emitterUpdateScript [
   &EmitterState(Life Cycle Mode=::NewEnumerator1, Loop Behavior=::NewEnumerator1, Loop Duration=1, UseLoopDelay=false)
   &SpawnBurst_Instantaneous(Spawn Count=10, Use Spawn Probability=false)
  ]
  particleSpawnScript [
   &InitializeParticle(
    Color Maximum=<rgba 0.0802 0.0823 0.0802 1>
    Color Minimum=<rgba 0.0497 0.0482 0.0497 1>
    Color Mode=::NewEnumerator2
    Lifetime Max=0.8
    Lifetime Min=0.6
    Lifetime Mode=::NewEnumerator1
    Mesh Scale Max=<0.05 0.05 0.05>
    Mesh Scale Min=<0.025 0.025 0.025>
    Mesh Scale Mode=::NewEnumerator2
    UsePositionOffset=false
    Write Material Random=true
   )
   &SphereLocation(
    MaskSpawnGroup=false
    Override Seed=false
    Sphere Radius=20
    Surface Only=false
    UseNonUniformScale=false
    UseOffset=false
    UseSphereOrientation=false
   )
   &AddVelocity(Velocity=&Multiply_Vector(
     A=$Output.SphereLocation.SphereNormal
     B=&VectorFromFloat(Value=&RandomRangeFloat(Minimum=50, Maximum=100))
    ))
   &InitialMeshOrientation(
    Rotation=&RandomRangeVector(Minimum=<0 0 0>, Maximum=<1 1 1>)
    Use Orientation Vector=true
    Use Rotation Vector=true
   )
  ]
  particleUpdateScript [
   &GravityForce(Gravity=<0 0 -400>)
   &MeshRotationForce(Rotation=&RandomRangeVector(Minimum=<-10 -10 -10>, Maximum=<10 10 10>))
   &SolveForcesAndVelocity(Clamp Velocity=false, Limit Acceleration=false)
   &SolveRotationalForcesAndVelocity()
   &ParticleState()
   &ScaleMeshSize(Scale Factor=&VectorFromFloat(Value=&FloatFromCurve(FloatCurve=DataInterface() {propertyValues {Curve {keys [<at 0 1>,<at 0.5 1>,<at 1 0>]}}})))
  ]
  rendererClasses [&NiagaraMeshRendererProperties]
  renderers [{rendererClass=&NiagaraMeshRendererProperties, rendererIndex=0}]
 }
 Wispy_Smoke {
  emitterUpdateScript [
   &EmitterState(Life Cycle Mode=::NewEnumerator1, Loop Behavior=::NewEnumerator1, Loop Duration=1, UseLoopDelay=false)
   &SpawnBurst_Instantaneous(Spawn Count=&UniformRangedInt(Minimum=8, Maximum=8), Use Spawn Probability=false)
  ]
  particleSpawnScript [
   /Niagara/Modules/Spawn/Initialization/InitializeParticle(
    Color=&UniformRangedLinearColor(Minimum=<rgba 0.5 0.5 0.5 0.4>, Maximum=<rgba 1 1 1 0.6>)
    Lifetime=&UniformRangedFloat(Minimum=1, Maximum=2)
    Module.Write Color=true
    Module.Write Lifetime=true
    Module.Write Mass=true
    Module.Write Material Random=true
    Module.Write Position=true
    Module.Write Scale=false
    Module.Write SpriteRotation=true
    Module.Write SpriteSize=true
    Sprite Rotation=&UniformRangedFloat(Minimum=0, Maximum=360)
    Sprite Size=&Vector2DFromFloat(Value=&UniformRangedFloat(Minimum=50, Maximum=60))
   )
   &Color(Color=&MultiplyLinearColor(
     Linear Color=$Particles.Color
     Scale Factor=&MakeVector4_FromLinearColor(LinearColor=&UniformRangedLinearColor(Minimum=<rgba 0.35 0.2778 0.21 1>, Maximum=<rgba 0.35 0.2567 0.21 1>))
    ))
   /SetVariables(Particles.UVScale:Vector2f=&UniformRangedVector2D(Minimum=<-1 1>, Maximum=<-1 1>)) {isSetParameters=true}
   &CameraOffset(Camera Offset Amount=10)
   &SphereLocation(
    MaskSpawnGroup=false
    Override Seed=false
    Sphere Radius=20
    Surface Only=false
    UseNonUniformScale=false
    UseOffset=false
    UseSphereOrientation=false
   )
   &AddVelocity(Velocity=&Multiply_Vector(
     A=$Output.SphereLocation.SphereNormal
     B=&VectorFromFloat(Value=&RandomRangeFloat(Minimum=50, Maximum=100))
    ))
  ]
  particleUpdateScript [
   &ParticleState()
   &Drag(Drag=&RandomRangeFloat(Minimum=2, Maximum=3), Use Linear Drag=true, Use Rotational Drag=false)
   &GravityForce(Gravity=<0 0 -200>)
   &SolveForcesAndVelocity(Clamp Velocity=false, Limit Acceleration=false)
   &ScaleColor(
    Scale Alpha=&FloatFromCurve(
     FloatCurve=DataInterface() {propertyValues {Curve {keys [
         {interpMode=RCIM_Cubic, tangentMode=RCTM_User, time=0, value=1}
         {arriveTangent=-1, interpMode=RCIM_Cubic, leaveTangent=-1, tangentMode=RCTM_User, time=0.0954, value=0.4177}
         <at 0.9979 0.0215>
        ]}}}
     Scale Curve=2
    )
    ScaleA=true
    ScaleRGB=true
   )
   &ScaleSpriteSize(Scale Factor=&Vector2DFromFloat(Value=&Lerp_Float(A=1, B=2, Alpha=$Particles.NormalizedAge)))
   /Niagara/Modules/Update/SubUV/SubUVAnimation(Number Of Frames=63)
   &DynamicMaterialParameters(
    Index 0 Param 1=1
    Index 0 Param 2=0.5
    Index 0 Param 3=1
    Index 0 Param 4=1
    Module.Param0WriteEnabled=true
    Module.Param1WriteEnabled=true
    Module.Param2WriteEnabled=true
    Module.Param3WriteEnabled=true
   )
  ]
  rendererClasses [&NiagaraSpriteRendererProperties]
  renderers [{rendererClass=&NiagaraSpriteRendererProperties}]
 }
}
summary {
 systemName=NS_Obsidian_Emote_DebrisExplosion
 emitters [
  {emitterName=Wispy_Smoke, rendererClasses [&NiagaraSpriteRendererProperties]}
  {emitterName=Rocks, rendererClasses [&NiagaraMeshRendererProperties]}
 ]
 userVariables []
}
