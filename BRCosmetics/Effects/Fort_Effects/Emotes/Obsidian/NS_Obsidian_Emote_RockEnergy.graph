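# Niagara system for the Obsidian emote's rock energy, showing soft purple glow sprites on the character's skeletal mesh.
# The only enabled emitter bursts one sprite per spawn group, pins each to the indexed bone or socket, and keeps it past its lifetime.
# Per-index arrays size the glows, a short age curve scales them in, and User.bIsFrontendPreview raises their alpha.
# Two mesh-based Portal emitters with similar spawn-group attachment are disabled.
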NiagaraSystem /BRCosmetics/Effects/Fort_Effects/Emotes/Obsidian/NS_Obsidian_Emote_RockEnergy
dependencies {
 usedDataInterfaces [&NiagaraDataInterfaceArrayFloat, &NiagaraDataInterfaceCurve]
 usedDynamicInputs [&MaskFloatBySpawnGroup]
 usedModules [
  &SystemState
  &self:SystemSpawnScript.SystemScriptSource.SystemScriptGraph.NiagaraNodeAssignment_0.SetVariables_#1
 ]
}
emitters {
 FresnelPortal001 {
  bEnabled=false
  emitterUpdateScript [
   &EmitterState(
    Inactive Response=::NewEnumerator1
    Life Cycle Mode=::NewEnumerator1
    Loop Behavior=::NewEnumerator1
    Loop Duration Mode=::NewEnumerator1
    UseLoopDelay=false
   )
   &SpawnBurst_Instantaneous(Spawn Count=24, Use Spawn Probability=false)
  ]
  particleSpawnScript [
   &FN_SetSpawnGroup(Number of Spawn Groups=24)
   &InitializeParticle(
    Color Mode=::NewEnumerator0
    Lifetime=5
    Mesh Scale Mode=::NewEnumerator3
    Mesh Uniform Scale=&SelectFloatFromArray(
     Direct Array Index=&SpawnGroup()
     Float Selection Array=DataInterface() {propertyValues {FloatData [1.0449,1.0514,1.0349,1.0537,1.0876,1.7054,0.7388,1.0424,4.8617,0.4548,2.0235,2.1181,1,1.0452,1.0538,1,1.0224,1,1.0359,1,1,1,1,1]}}
    )
    UsePositionOffset=false
    Write Material Random=true
   )
   &SubUVAnimation(End Frame:NiagaraInt32=&SpawnGroup(), Start Frame:NiagaraInt32=&SpawnGroup()) {module=SubUVAnimation001}
   &InitialMeshOrientation(Orientation Axis=<0 1 0>, Use Orientation Vector=true, Use Rotation Vector=false) {enabled=false}
   &CameraOffset(Camera Offset Amount=-20)
  ]
  particleUpdateScript [
   &ParticleState(Kill Particles When Lifetime Has Elapsed=false, Let Infinitely Lived Particles Die When Emitter Deactivates=true)
   &SkeletalMeshLocation(
    Apply Sampled Orientation to Mesh Particles=true
    Bone / Socket Index=&SpawnGroup()
    Bone Sampling Mode=::NewEnumerator4
    Inherit Velocity=false
    MaskSpawnGroup=false
    Skeletal Mesh=$User.SkeletalMesh
   )
   /SetVariables(Particles.MeshOrientation=$Output.SkeletalMeshLocation.SampledRotation) {enabled=false, isSetParameters=true}
   &ScaleColor(
    Scale Alpha=&MakeCustomFloatFromBool(
     Bool=&SetBoolByIntComparison(A=&SpawnGroup(), B=3, Comparison Type=::NewEnumerator2)
     True Float=1
     False Float=0
    )
    ScaleA=true
    ScaleRGB=true
   )
   /SetVariables(
    Particles.GlowColor=&MultiplyLinearColor(Linear Color=$Particles.Color, Scale Factor=<1 1 1 0.5>)
    Particles.GlowSize:Vector3f=&Multiply_Vector(A=$Particles.Scale, B=&VectorFromFloat(Value=0.9))
   ) {isSetParameters=true, module=SetVariables_#2}
   &ScaleMeshSize(Scale Factor=&VectorFromFloat(Value=0.7764))
   &ScaleMeshSize(Scale Factor=&VectorFromFloat(Value=&FloatFromCurve(FloatCurve=DataInterface() {propertyValues {Curve {keys [<at 0 0>,<at 0.2 1>]}}}, CurveIndex=$Particles.Age))) {module=ScaleMeshSize001}
  ]
  rendererClasses [&NiagaraMeshRendererProperties]
  renderers [
   {rendererClass=&NiagaraMeshRendererProperties, rendererIndex=0}
   {rendererClass=&NiagaraMeshRendererProperties, rendererIndex=1}
  ]
 }
 Portal {
  bEnabled=false
  emitterUpdateScript [
   &EmitterState(
    Inactive Response=::NewEnumerator1
    Life Cycle Mode=::NewEnumerator1
    Loop Behavior=::NewEnumerator1
    Loop Duration Mode=::NewEnumerator1
    UseLoopDelay=false
   )
   &SpawnBurst_Instantaneous(Spawn Count=27, Use Spawn Probability=false)
  ]
  particleSpawnScript [
   &FN_SetSpawnGroup(Number of Spawn Groups=27)
   &InitializeParticle(
    Color Mode=::NewEnumerator0
    Lifetime=5
    Mesh Scale Mode=::NewEnumerator3
    Mesh Uniform Scale=&SelectFloatFromArray(
     Direct Array Index=&SpawnGroup()
     Float Selection Array=DataInterface() {propertyValues {FloatData [0,0,0,0.6,0.7,0.7,1.1,0.9,0.7,0.5,1.1,1.6,0.4,0.8,0.8,0.6,0.6,1,0.6,1.6,0.4,0.8,0.7,0.7,0.5,0.4,0.7]}}
    )
    UsePositionOffset=false
    Write Material Random=true
   )
   &InitialMeshOrientation(Orientation Axis=<0 1 0>, Use Orientation Vector=true, Use Rotation Vector=false)
   &CameraOffset(Camera Offset Amount=-70)
  ]
  particleUpdateScript [
   &ParticleState(Kill Particles When Lifetime Has Elapsed=false, Let Infinitely Lived Particles Die When Emitter Deactivates=true)
   &SkeletalMeshLocation(
    Bone / Socket Index=&SpawnGroup()
    Bone Sampling Mode=::NewEnumerator6
    Inherit Velocity=false
    MaskSpawnGroup=false
    Skeletal Mesh=$User.SkeletalMesh
   )
   &ScaleMeshSize(Scale Factor=&VectorFromFloat(Value=&FloatFromCurve(
      FloatCurve=DataInterface() {propertyValues {Curve {keys [<at 0 0 RCIM_Cubic>, {arriveTangent=14.4618, interpMode=RCIM_Cubic, leaveTangent=14.4618, tangentMode=RCTM_User, time=0.15, value=1}]}}}
      CurveIndex=$Particles.Age
     )))
   &ScaleColor(
    Scale Alpha=&MakeCustomFloatFromBool(
     Bool=&SetBoolByIntComparison(A=&SpawnGroup(), B=3, Comparison Type=::NewEnumerator2)
     True Float=1
     False Float=0
    )
    ScaleA=true
    ScaleRGB=true
   )
   &ScaleMeshSize(Scale Factor=&RandomRangeVector(Minimum=<0.8 0.8 0.8>, Maximum=<1.2 1.2 1.2>)) {module=ScaleMeshSize001}
   /SetVariables(
    Particles.GlowColor=&MultiplyLinearColor(Linear Color=$Particles.Color, Scale Factor=<1 1 1 0.5>)
    Particles.GlowSize:Vector3f=&Multiply_Vector(A=$Particles.Scale, B=&VectorFromFloat(Value=0.9))
   ) {isSetParameters=true, module=SetVariables_#2}
  ]
  rendererClasses [&NiagaraMeshRendererProperties]
  renderers [
   {rendererClass=&NiagaraMeshRendererProperties, rendererIndex=0}
   {rendererClass=&NiagaraMeshRendererProperties, rendererIndex=1}
   {rendererClass=&NiagaraMeshRendererProperties, rendererIndex=2}
   {rendererClass=&NiagaraMeshRendererProperties, rendererIndex=3}
   {rendererClass=&NiagaraMeshRendererProperties, rendererIndex=4}
  ]
 }
 Wrist_Airglow001 {
  emitterUpdateScript [
   &EmitterState(
    Inactive Response=::NewEnumerator1
    Life Cycle Mode=::NewEnumerator1
    Loop Behavior=::NewEnumerator1
    Loop Duration Mode=::NewEnumerator1
    UseLoopDelay=false
   )
   &SpawnBurst_Instantaneous(Spawn Count=27, Use Spawn Probability=false)
  ]
  particleSpawnScript [
   &FN_SetSpawnGroup(Number of Spawn Groups=27)
   /Niagara/Modules/Spawn/Initialization/InitializeParticle(
    Color=&MakeCustomLinearColorFromBool(Bool=$User.bIsFrontendPreview, True Color=<rgba 0.6924 0.0726 1 0.6>, False Color=<rgba 0.6924 0.0726 1 0.15>)
    Lifetime=0.25
    Module.Write Color=true
    Module.Write Lifetime=true
    Module.Write Mass=false
    Module.Write Material Random=true
    Module.Write Position=false
    Module.Write Scale=false
    Module.Write SpriteRotation=false
    Module.Write SpriteSize=true
    Position=&SimulationPosition()
    Sprite Size=&Vector2DFromFloat(Value=35)
   )
   &DynamicMaterialParameters(
    Index 0 Param 1=1
    Module.Param0WriteEnabled=true
    Module.Param1WriteEnabled=true
    Module.Param2WriteEnabled=true
    Module.Param3WriteEnabled=true
   )
  ]
  particleUpdateScript [
   &ParticleState(Kill Particles When Lifetime Has Elapsed=false, Let Infinitely Lived Particles Die When Emitter Deactivates=true)
   &SkeletalMeshLocation(
    Bone / Socket Index=&SpawnGroup()
    Bone Sampling Mode=::NewEnumerator6
    Inherit Velocity=false
    MaskSpawnGroup=false
    Skeletal Mesh=$User.SkeletalMesh
   )
   &CameraOffset(Camera Offset Amount=-15)
   &ScaleColor(Scale Alpha=0.1, Scale RGB=&VectorFromFloat(Value=0.8), ScaleA=true, ScaleRGB=true)
   &ScaleSpriteSize(Scale Factor=&Vector2DFromFloat(Value=&FloatFromCurve(
      FloatCurve=DataInterface() {propertyValues {Curve {keys [<at 0 0 RCIM_Cubic>, {arriveTangent=14.8605, interpMode=RCIM_Cubic, leaveTangent=14.8605, tangentMode=RCTM_User, time=0.15, value=1}]}}}
      CurveIndex=$Particles.Age
     )))
   &ScaleSpriteSize(Scale Factor=&Vector2DFromFloat(Value=&SelectFloatFromArray(
      Direct Array Index=&SpawnGroup()
      Float Selection Array=DataInterface() {propertyValues {FloatData [0,0,0,0.6,0.7,0.7,1.1,0.9,0.7,0.5,1.1,1.6,0.4,0.8,0.8,0.6,0.6,1,0.6,1.6,0.4,0.8,0.7,0.7,0.5,0.4,0.7]}}
     ))) {module=ScaleSpriteSize001}
  ]
  rendererClasses [&NiagaraSpriteRendererProperties]
  renderers [{rendererClass=&NiagaraSpriteRendererProperties}]
 }
}
summary {
 systemName=NS_Obsidian_Emote_RockEnergy
 emitters [
  {bEnabled=false, emitterName=Portal, rendererClasses [&NiagaraMeshRendererProperties]}
  {emitterName=Wrist_Airglow001, rendererClasses [&NiagaraSpriteRendererProperties]}
  {bEnabled=false, emitterName=FresnelPortal001, rendererClasses [&NiagaraMeshRendererProperties]}
 ]
 userVariables [
  User.bIsFrontendPreview {type=NiagaraBool}
  User.SkeletalMesh {
   type=NiagaraDataInterfaceSkeletalMesh
   defaultValue {
    struct=NiagaraExt_VariableValue_DataInterface
    value {dataInterface=&self:NiagaraDataInterfaceSkeletalMesh_0}
   }
  }
 ]
}
