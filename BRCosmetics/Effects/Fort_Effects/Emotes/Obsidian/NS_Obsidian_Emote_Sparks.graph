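# Niagara system for the Obsidian emote, throwing dark rock debris and brief purple light-casting flares off a user-supplied skeletal mesh.
# Shaking emitters ramp flare and small-rock spawn rates starting about five seconds in, overlapping timed bursts around the six-second mark.
# Large-rock and flare bursts use a random spawn group as bone/socket index.
# Tumbling mesh rocks launch along sampled bone orientation under gravity, with smaller debris simulated on GPU.
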NiagaraSystem /BRCosmetics/Effects/Fort_Effects/Emotes/Obsidian/NS_Obsidian_Emote_Sparks
dependencies {
 usedDataInterfaces [&NiagaraDataInterfaceCurve]
 usedDynamicInputs []
 usedModules [
  &SystemState
  &self:SystemSpawnScript.SystemScriptSource.SystemScriptGraph.NiagaraNodeAssignment_1.SetVariables_#1
 ]
}
emitters {
 Flare_Bursts {
  emitterUpdateScript [
   &EmitterState(Life Cycle Mode=::NewEnumerator1, Loop Behavior=::NewEnumerator1, Loop Duration=20, UseLoopDelay=false)
   &SpawnBurst_Instantaneous(Spawn Count=1, Spawn Group:NiagaraInt32=$System.RandomSpawnGroup, Spawn Time=6, Use Spawn Probability=false)
   &SpawnBurst_Instantaneous(Spawn Count=1, Spawn Group:NiagaraInt32=$System.RandomSpawnGroup, Spawn Time=6.05, Use Spawn Probability=false) {module=SpawnBurst_Instantaneous001}
   &SpawnBurst_Instantaneous(Spawn Count=1, Spawn Group:NiagaraInt32=$System.RandomSpawnGroup, Spawn Time=6.1, Use Spawn Probability=false) {module=SpawnBurst_Instantaneous002}
   &SpawnBurst_Instantaneous(Spawn Count=1, Spawn Group:NiagaraInt32=$System.RandomSpawnGroup, Spawn Time=6.15, Use Spawn Probability=false) {module=SpawnBurst_Instantaneous003}
   &SpawnBurst_Instantaneous(Spawn Count=1, Spawn Group:NiagaraInt32=$System.RandomSpawnGroup, Spawn Time=6.2, Use Spawn Probability=false) {module=SpawnBurst_Instantaneous004}
   &SpawnBurst_Instantaneous(Spawn Count=1, Spawn Group:NiagaraInt32=$System.RandomSpawnGroup, Spawn Time=6.25, Use Spawn Probability=false) {module=SpawnBurst_Instantaneous005}
  ]
  particleSpawnScript [
   &SkeletalMeshLocation(
    Bone / Socket Index=&SpawnGroup()
    Bone Sampling Mode=::NewEnumerator4
    Inherit Velocity=false
    MaskSpawnGroup=false
    Skeletal Mesh=$User.SkeletalMesh
   )
   /Niagara/Modules/Spawn/Initialization/InitializeParticle(
    Color=<rgba 0.6924 0.0726 1 1>
    Lifetime=0.1
    Module.Write Color=true
    Module.Write Lifetime=true
    Module.Write Mass=true
    Module.Write Material Random=true
    Module.Write Position=false
    Module.Write Scale=false
    Module.Write SpriteRotation=false
    Module.Write SpriteSize=true
    Position=&SimulationPosition()
    Sprite Size=<30 30>
   )
   &CameraOffset(Camera Offset Amount=50)
   &DynamicMaterialParameters(
    Index 0 Param 1=0.2
    Index 0 Param 2=1
    Module.Param0WriteEnabled=true
    Module.Param1WriteEnabled=true
    Module.Param2WriteEnabled=true
    Module.Param3WriteEnabled=true
   )
  ]
  particleUpdateScript [
   &ParticleState()
   &SolveForcesAndVelocity(Clamp Velocity=false, Limit Acceleration=false)
   &ScaleColor(
    Scale Alpha=&Lerp_Float(A=1, B=0, Alpha=$Particles.NormalizedAge)
    Scale RGB=&VectorFromFloat(Value=4)
    ScaleA=true
    ScaleRGB=true
   )
   &FN_Light_v2(Light Exponent=3, Light Radius=150, LightEnabled=true, Scale Color=1)
  ]
  rendererClasses [&NiagaraSpriteRendererProperties, &NiagaraLightRendererProperties]
  renderers [
   {rendererClass=&NiagaraSpriteRendererProperties}
   {rendererClass=&NiagaraLightRendererProperties, rendererIndex=1}
  ]
 }
 Flare_Shaking {
  emitterUpdateScript [
   &EmitterState(Life Cycle Mode=::NewEnumerator1, Loop Behavior=::NewEnumerator1, Loop Delay=5.61, Loop Duration=1, UseLoopDelay=true)
   &SpawnRate(
    SpawnRate=&FloatFromCurve(FloatCurve=DataInterface() {propertyValues {Curve {keys [<at 0 0>,<at 1 20>]}}}, CurveIndex=$Emitter.NormalizedLoopAge)
    Use Spawn Probability=false
   )
  ]
  particleSpawnScript [
   &SkeletalMeshLocation(
    Bone Sampling Mode=::NewEnumerator0
    Inherit Velocity=false
    MaskSpawnGroup=false
    Module.Override Randomness=false
    Skeletal Mesh=$User.SkeletalMesh
   )
   &InitializeParticle(
    Color=<rgba 0.6924 0.0726 1 1>
    Lifetime=0.1
    Module.Write Color=true
    Module.Write Lifetime=true
    Module.Write Mass=true
    Module.Write Material Random=true
    Module.Write Position=false
    Module.Write Scale=false
    Module.Write SpriteRotation=false
    Module.Write SpriteSize=true
    Position=&SimulationPosition()
    Sprite Size=<30 30>
   )
   &CameraOffset(Camera Offset Amount=50)
   &DynamicMaterialParameters(
    Index 0 Param 1=0.2
    Index 0 Param 2=1
    Module.Param0WriteEnabled=true
    Module.Param1WriteEnabled=true
    Module.Param2WriteEnabled=true
    Module.Param3WriteEnabled=true
   )
  ]
  particleUpdateScript [
   &ParticleState()
   &SolveForcesAndVelocity(Clamp Velocity=false, Limit Acceleration=false)
   &ScaleColor(
    Scale Alpha=&Lerp_Float(A=1, B=0, Alpha=$Particles.NormalizedAge)
    Scale RGB=&VectorFromFloat(Value=4)
    ScaleA=true
    ScaleRGB=true
   )
   &FN_Light_v2(Light Exponent=3, Light Radius=150, LightEnabled=true, Scale Color=1)
  ]
  rendererClasses [&NiagaraSpriteRendererProperties, &NiagaraLightRendererProperties]
  renderers [
   {rendererClass=&NiagaraSpriteRendererProperties}
   {rendererClass=&NiagaraLightRendererProperties, rendererIndex=1}
  ]
 }
 LargeRocks_Bursts {
  emitterUpdateScript [
   &EmitterState(Life Cycle Mode=::NewEnumerator1, Loop Behavior=::NewEnumerator1, Loop Duration=20, UseLoopDelay=false)
   &SpawnBurst_Instantaneous(
    Spawn Count=&RandomRangeInt(Minimum=1, Maximum=10)
    Spawn Group:NiagaraInt32=$System.RandomSpawnGroup
    Spawn Time=6
    Use Spawn Probability=false
   )
   &SpawnBurst_Instantaneous(
    Spawn Count=&RandomRangeInt(Minimum=1, Maximum=10)
    Spawn Group:NiagaraInt32=$System.RandomSpawnGroup
    Spawn Time=6.05
    Use Spawn Probability=false
   ) {module=SpawnBurst_Instantaneous001}
   &SpawnBurst_Instantaneous(
    Spawn Count=&RandomRangeInt(Minimum=1, Maximum=10)
    Spawn Group:NiagaraInt32=$System.RandomSpawnGroup
    Spawn Time=6.1
    Use Spawn Probability=false
   ) {module=SpawnBurst_Instantaneous002}
   &SpawnBurst_Instantaneous(
    Spawn Count=&RandomRangeInt(Minimum=1, Maximum=10)
    Spawn Group:NiagaraInt32=$System.RandomSpawnGroup
    Spawn Time=6.15
    Use Spawn Probability=false
   ) {module=SpawnBurst_Instantaneous003}
   &SpawnBurst_Instantaneous(
    Spawn Count=&RandomRangeInt(Minimum=1, Maximum=10)
    Spawn Group:NiagaraInt32=$System.RandomSpawnGroup
    Spawn Time=6.2
    Use Spawn Probability=false
   ) {module=SpawnBurst_Instantaneous004}
   &SpawnBurst_Instantaneous(
    Spawn Count=&RandomRangeInt(Minimum=1, Maximum=10)
    Spawn Group:NiagaraInt32=$System.RandomSpawnGroup
    Spawn Time=6.25
    Use Spawn Probability=false
   ) {module=SpawnBurst_Instantaneous005}
  ]
  particleSpawnScript [
   &SkeletalMeshLocation(
    Bone / Socket Index=&SpawnGroup()
    Bone Sampling Mode=::NewEnumerator4
    Inherit Velocity=false
    MaskSpawnGroup=false
    Skeletal Mesh=$User.SkeletalMesh
   )
   /SetVariables(Particles.Distribution:NiagaraFloat=&Add_Float(
     A=&UniformAOrBFloat(A=0.25, B=0.9, Distribution Weight=0.1)
     B=&RandomRangeFloat(Minimum=-0.1, Maximum=0.1)
    )) {isSetParameters=true, module=SetVariables_#2}
   &InitializeParticle(
    Color=<rgba 0.0125 0.0099 0.0068 1>
    Lifetime=&UniformRangedFloat(Minimum=0.4, Maximum=0.8)
    Mesh Scale=&VectorFromFloat(Value=/Game/Effects/Niagara/DynamicInputs/Math/FN_Fit01(Clamp:NiagaraBool=false, Value:NiagaraFloat=$Particles.Distribution, DestMin:NiagaraFloat=0.02, DestMax:NiagaraFloat=0.04))
    Module.Write Color=true
    Module.Write Lifetime=true
    Module.Write Mass=true
    Module.Write Material Random=true
    Module.Write Position=false
    Module.Write Scale=true
    Module.Write SpriteRotation=false
    Module.Write SpriteSize=false
    Position=&SimulationPosition()
    Sprite Size=&UniformRangedVector2D(Minimum=<2.1064 4.7118>, Maximum=<4.4317 8.5012>)
   )
   &InitialMeshOrientation(Orientation Vector=&RandomVector(Vector Scale=1), Use Orientation Vector=true, Use Rotation Vector=false)
   &AddVelocityInCone(
    Cone Axis=&RotateVectorByQuaternion(Vector To Rotate=<1 0 0>, Quaternion=$Output.SkeletalMeshLocation.SampledRotation)
    Module.Use Velocity Falloff On Cone Axis=false
    Velocity Strength=&RandomRangeFloat(Minimum=50, Maximum=200)
   )
   &MeshRotationForce(Rotation=&RandomVector(Vector Scale=&RandomRangeFloat(Minimum=5, Maximum=10)))
   &ApplyInitialForces()
  ]
  particleUpdateScript [
   &ParticleState()
   &AccelerationForce(Acceleration=&MakeVector(X=0, Y=0, Z=-980))
   &Drag(Drag=0.1, Use Linear Drag=true, Use Rotational Drag=false)
   &SolveForcesAndVelocity(Clamp Velocity=false, Limit Acceleration=false)
   &SolveRotationalForcesAndVelocity()
   &FN_Mesh_Attributes(
    Create Scale Attribute=true
    Create Visibility Attribute=true
    Scale=&Multiply_Vector(
     A=$Particles.Initial.Scale
     B=&VectorFromFloat(Value=&FloatFromCurve(FloatCurve=DataInterface() {propertyValues {Curve {keys [
           <at -0.0032 0 RCIM_Cubic>
           {arriveTangent=-0.0978, interpMode=RCIM_Cubic, leaveTangent=-0.0978, tangentMode=RCTM_User, time=0.1249, value=0.9962}
           {arriveTangent=0.0267, interpMode=RCIM_Cubic, leaveTangent=0.0267, tangentMode=RCTM_User, time=0.7407, value=1.0021}
           <at 1 0 RCIM_Cubic>
          ]}}}))
    )
   )
  ]
  rendererClasses [&NiagaraMeshRendererProperties]
  renderers [{rendererClass=&NiagaraMeshRendererProperties, rendererIndex=0}]
 }
 LittleRock_Bursts {
  simTarget=GPUComputeSim
  emitterUpdateScript [
   &EmitterState(Life Cycle Mode=::NewEnumerator1, Loop Behavior=::NewEnumerator1, Loop Delay=0.4, Loop Duration=20, UseLoopDelay=false)
   &SpawnBurst_Instantaneous(
    Spawn Count=&RandomRangeInt(Minimum=10, Maximum=30)
    Spawn Probability=0
    Spawn Time=6
    Use Spawn Probability=false
   )
   &SpawnBurst_Instantaneous(
    Spawn Count=&RandomRangeInt(Minimum=10, Maximum=30)
    Spawn Probability=0.4
    Spawn Time=6.15
    Use Spawn Probability=false
   ) {module=SpawnBurst_Instantaneous001}
   &SpawnBurst_Instantaneous(
    Spawn Count=&RandomRangeInt(Minimum=40, Maximum=60)
    Spawn Probability=0.4
    Spawn Time=6.6
    Use Spawn Probability=false
   ) {module=SpawnBurst_Instantaneous002}
  ]
  particleSpawnScript [
   &SkeletalMeshLocation(
    Bone Sampling Mode=::NewEnumerator0
    Inherit Velocity=false
    MaskSpawnGroup=false
    Module.Override Randomness=false
    Skeletal Mesh=$User.SkeletalMesh
   )
   /SetVariables(Particles.Distribution:NiagaraFloat=&Add_Float(
     A=&UniformAOrBFloat(A=0.25, B=0.9, Distribution Weight=0.1)
     B=&RandomRangeFloat(Minimum=-0.1, Maximum=0.1)
    )) {isSetParameters=true, module=SetVariables_#2}
   &InitializeParticle(
    Color=<rgba 0.0125 0.0099 0.0068 1>
    Lifetime=&UniformRangedFloat(Minimum=0.4, Maximum=0.8)
    Mesh Scale=&VectorFromFloat(Value=&FN_Fit01(Clamp:NiagaraBool=false, Value:NiagaraFloat=$Particles.Distribution, DestMin:NiagaraFloat=0.005, DestMax:NiagaraFloat=0.01))
    Module.Write Color=true
    Module.Write Lifetime=true
    Module.Write Mass=true
    Module.Write Material Random=true
    Module.Write Position=false
    Module.Write Scale=true
    Module.Write SpriteRotation=false
    Module.Write SpriteSize=false
    Position=&SimulationPosition()
    Sprite Size=&UniformRangedVector2D(Minimum=<2.1064 4.7118>, Maximum=<4.4317 8.5012>)
   )
   &InitialMeshOrientation(Orientation Vector=&RandomVector(Vector Scale=1), Use Orientation Vector=true, Use Rotation Vector=false)
   &AddVelocityInCone(
    Cone Angle=30
    Cone Axis=&RotateVectorByQuaternion(Vector To Rotate=<1 0 0>, Quaternion=$Output.SkeletalMeshLocation.SampledRotation)
    Module.Use Velocity Falloff On Cone Axis=false
    Velocity Strength=&RandomRangeFloat(Minimum=100, Maximum=200)
   )
   &MeshRotationForce(Rotation=&RandomVector(Vector Scale=&RandomRangeFloat(Minimum=5, Maximum=10)))
   &ApplyInitialForces()
  ]
  particleUpdateScript [
   &ParticleState()
   &AccelerationForce(Acceleration=&MakeVector(X=0, Y=0, Z=-450))
   &Drag(Drag=0.1, Use Linear Drag=true, Use Rotational Drag=false)
   &SolveForcesAndVelocity(Clamp Velocity=false, Limit Acceleration=false)
   &SolveRotationalForcesAndVelocity()
   &FN_Mesh_Attributes(
    Create Scale Attribute=true
    Create Visibility Attribute=true
    Scale=&Multiply_Vector(
     A=$Particles.Initial.Scale
     B=&VectorFromFloat(Value=&FloatFromCurve(FloatCurve=DataInterface() {propertyValues {Curve {keys [
           <at -0.0032 0 RCIM_Cubic>
           {arriveTangent=-0.0978, interpMode=RCIM_Cubic, leaveTangent=-0.0978, tangentMode=RCTM_User, time=0.1249, value=0.9962}
           {arriveTangent=0.0267, interpMode=RCIM_Cubic, leaveTangent=0.0267, tangentMode=RCTM_User, time=0.7407, value=1.0021}
           <at 1 0 RCIM_Cubic>
          ]}}}))
    )
   )
  ]
  rendererClasses [&NiagaraMeshRendererProperties]
  renderers [{rendererClass=&NiagaraMeshRendererProperties, rendererIndex=0}]
 }
 SmallRocks_Shaking {
  simTarget=GPUComputeSim
  emitterUpdateScript [
   &EmitterState(Life Cycle Mode=::NewEnumerator1, Loop Behavior=::NewEnumerator1, Loop Delay=5.45, Loop Duration=1.15, UseLoopDelay=true)
   &SpawnRate(
    SpawnRate=&FloatFromCurve(FloatCurve=DataInterface() {propertyValues {Curve {keys [<at 0 20>,<at 1 50>]}}}, CurveIndex=$Emitter.NormalizedLoopAge)
    Use Spawn Probability=false
   )
  ]
  particleSpawnScript [
   &SkeletalMeshLocation(
    Bone Sampling Mode=::NewEnumerator0
    Inherit Velocity=false
    MaskSpawnGroup=false
    Module.Override Randomness=false
    Skeletal Mesh=$User.SkeletalMesh
   )
   /SetVariables(Particles.Distribution:NiagaraFloat=&Add_Float(
     A=&UniformAOrBFloat(A=0.25, B=0.9, Distribution Weight=0.1)
     B=&RandomRangeFloat(Minimum=-0.1, Maximum=0.1)
    )) {isSetParameters=true, module=SetVariables_#2}
   &InitializeParticle(
    Color=<rgba 0.0125 0.0099 0.0068 1>
    Lifetime=&UniformRangedFloat(Minimum=0.4, Maximum=0.8)
    Mesh Scale=&VectorFromFloat(Value=&FN_Fit01(Clamp:NiagaraBool=false, Value:NiagaraFloat=$Particles.Distribution, DestMin:NiagaraFloat=0.01, DestMax:NiagaraFloat=0.02))
    Module.Write Color=true
    Module.Write Lifetime=true
    Module.Write Mass=true
    Module.Write Material Random=true
    Module.Write Position=false
    Module.Write Scale=true
    Module.Write SpriteRotation=false
    Module.Write SpriteSize=false
    Position=&SimulationPosition()
    Sprite Size=&UniformRangedVector2D(Minimum=<2.1064 4.7118>, Maximum=<4.4317 8.5012>)
   )
   &InitialMeshOrientation(Orientation Vector=&RandomVector(Vector Scale=1), Use Orientation Vector=true, Use Rotation Vector=false)
   &AddVelocityInCone(
    Cone Axis=&RotateVectorByQuaternion(Vector To Rotate=<1 0 0>, Quaternion=$Output.SkeletalMeshLocation.SampledRotation)
    Module.Use Velocity Falloff On Cone Axis=false
    Velocity Strength=&RandomRangeFloat(Minimum=25, Maximum=50)
   )
   &MeshRotationForce(Rotation=&RandomVector(Vector Scale=&RandomRangeFloat(Minimum=5, Maximum=10)))
   &ApplyInitialForces()
  ]
  particleUpdateScript [
   &ParticleState()
   &AccelerationForce(Acceleration=&MakeVector(X=0, Y=0, Z=-600))
   &Drag(Drag=0.1, Use Linear Drag=true, Use Rotational Drag=false)
   &SolveForcesAndVelocity(Clamp Velocity=false, Limit Acceleration=false)
   &SolveRotationalForcesAndVelocity()
   &FN_Mesh_Attributes(
    Create Scale Attribute=true
    Create Visibility Attribute=true
    Scale=&Multiply_Vector(
     A=$Particles.Initial.Scale
     B=&VectorFromFloat(Value=&FloatFromCurve(FloatCurve=DataInterface() {propertyValues {Curve {keys [
           <at -0.0032 0 RCIM_Cubic>
           {arriveTangent=-0.0978, interpMode=RCIM_Cubic, leaveTangent=-0.0978, tangentMode=RCTM_User, time=0.1249, value=0.9962}
           {arriveTangent=0.0267, interpMode=RCIM_Cubic, leaveTangent=0.0267, tangentMode=RCTM_User, time=0.7407, value=1.0021}
           <at 1 0 RCIM_Cubic>
          ]}}}))
    )
   )
  ]
  rendererClasses [&NiagaraMeshRendererProperties]
  renderers [{rendererClass=&NiagaraMeshRendererProperties, rendererIndex=0}]
 }
}
summary {
 systemName=NS_Obsidian_Emote_Sparks
 emitters [
  {emitterName=LargeRocks_Bursts, rendererClasses [&NiagaraMeshRendererProperties]}
  {emitterName=Flare_Bursts, rendererClasses [&NiagaraSpriteRendererProperties, &NiagaraLightRendererProperties]}
  {emitterName=Flare_Shaking, rendererClasses [&NiagaraSpriteRendererProperties, &NiagaraLightRendererProperties]}
  {emitterName=SmallRocks_Shaking, simTarget=GPUComputeSim, rendererClasses [&NiagaraMeshRendererProperties]}
  {emitterName=LittleRock_Bursts, simTarget=GPUComputeSim, rendererClasses [&NiagaraMeshRendererProperties]}
 ]
 userVariables [User.SkeletalMesh {
   type=NiagaraDataInterfaceSkeletalMesh
   defaultValue {
    struct=NiagaraExt_VariableValue_DataInterface
    value {dataInterface=&self:NiagaraDataInterfaceSkeletalMesh_1}
   }
  }]
}
